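# Sound cue for the fire pit hazard's eruption burst, layering a randomized eruption with a delayed rattle.
# A mixer plays one of five pitch-modulated eruptions alongside a slightly quieter, randomized rattle with pitch and volume variation, delayed half a second.
# The rattle routes through a quality node that feeds only the first two of four levels, so it drops out on the others.
# Playback uses the Traps sound class, custom attenuation, limited concurrency and no virtualization.
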SoundCue /FirePetal_WanderingBoss/Assets/Audio/FirePitHazard_Erupt_Burst_Cue
root=Mixer_2
nodes [
 {id=Delay_0, children [Modulator_4], properties {delayMax=0.5, delayMin=0.5}}
 {id=Mixer_2, children [Modulator_3, QualityLevel_0], properties {inputVolume [1,0.8]}}
 {id=Modulator_3, children [Random_0], properties {pitchMax=1.2, pitchMin=0.9}}
 {id=Modulator_4, children [Random_1], properties {pitchMax=1.2, volumeMax=1, volumeMin=0.7}}
 {id=QualityLevel_0, children [Delay_0,Delay_0,null,null]}
 {id=Random_0, children [WavePlayer_0, WavePlayer_1, WavePlayer_2, WavePlayer_3, WavePlayer_18], properties {weights [1,1,1,1,1]}}
 {
  id=Random_1
  children [WavePlayer_5, WavePlayer_6, WavePlayer_7, WavePlayer_16, WavePlayer_17]
  properties {bShouldExcludeFromBranchCulling=true, weights [1,1,1,1,1]}
 }
 {id=WavePlayer_0, properties {soundWaveAssetPtr=&/FirePitHazard_Erupt_01}}
 {id=WavePlayer_1, properties {soundWaveAssetPtr=&/FirePitHazard_Erupt_02}}
 {id=WavePlayer_16, properties {soundWaveAssetPtr=&/FirePitHazard_Erupt_Rattle_04}}
 {id=WavePlayer_17, properties {soundWaveAssetPtr=&/FirePitHazard_Erupt_Rattle_05}}
 {id=WavePlayer_18, properties {soundWaveAssetPtr=&/FirePitHazard_Erupt_05}}
 {id=WavePlayer_2, properties {soundWaveAssetPtr=&/FirePitHazard_Erupt_03}}
 {id=WavePlayer_3, properties {soundWaveAssetPtr=&/FirePitHazard_Erupt_04}}
 {id=WavePlayer_5, properties {soundWaveAssetPtr=&/FirePitHazard_Erupt_Rattle_01}}
 {id=WavePlayer_6, properties {soundWaveAssetPtr=&/FirePitHazard_Erupt_Rattle_02}}
 {id=WavePlayer_7, properties {soundWaveAssetPtr=&/FirePitHazard_Erupt_Rattle_03}}
]
properties {
 attenuationSettings=&/FirePitHazard_Erupt_Burst_Att
 bOverrideConcurrency=true
 maxDistance=6000
 soundClassObject=&Fort_Traps
 virtualizationMode=Disabled
 volumeMultiplier=0.85
 attenuationOverrides {bEnableListenerFocus=true, falloffDistance=2000, attenuationShapeExtents {x=500}}
 concurrencyOverrides {maxCount=3}
}
