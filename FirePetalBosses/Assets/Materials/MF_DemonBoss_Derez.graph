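# Material function that adds an animated demon-boss derez dissolve to incoming material attributes, or passes them through when UseDerezMaterial is off.
# Pre-skinned local position, carried to the pixel shader in customized UVs, builds a bounds-relative axis gradient and feeds distorted, panning noise.
# A phase from a custom-primitive-data timestamp or manual dissolve parameter sweeps threshold masks driving base color, two-tone emissive, opacity mask and normal-plus-smear vertex offset.
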MaterialFunction /FirePetalBosses/Assets/Materials/MF_DemonBoss_Derez
graph {nodes [
  n1 = NamedRerouteUsage()
  n2 = SetMaterialAttributes!(
   MaterialAttributes=@n1
   BaseColor=LinearInterpolate(A=NamedRerouteUsage(), B=Constant3Vector(), Alpha=NamedRerouteUsage())
   EmissiveColor=NamedRerouteUsage()
   OpacityMask=OneMinus(None=NamedRerouteUsage())
   WorldPositionOffset=Add(A=NamedRerouteUsage(), B=NamedRerouteUsage())
   Roughness
   Specular
   CustomizedUV1=NamedRerouteUsage()
   CustomizedUV2=NamedRerouteUsage()
  ) {attributeSetTypes [
    #1
    #2
    #3
    #4
    #5
    #6
    #7
    #8
   ]}
  FunctionOutput(None=StaticSwitchParameter(True=@n2, False=@n1) {group=Boss_Derez, parameterName=UseDerezMaterial})
  n3 = GetMaterialAttributes(None=FunctionInput() {inputName=Material, inputType=FunctionInput_MaterialAttributes}) {attributeGetTypes [
    #4
    #1
    #6
    #5
   ]}
  NamedRerouteDeclaration(None=@n3.MaterialAttributes) {name=Input_Material, nodeColor=<rgba 0 1 0.5313 1>}
  n4 = TextureCoordinate() {coordinateIndex=2}
  n5 = NamedRerouteUsage()
  n6 = ShaderStageSwitch(
   PixelShader=AppendVector(A=TextureCoordinate() {coordinateIndex=1}, B=ComponentMask(None=@n4) {r=true})
   VertexShader=AppendVector(A=NamedRerouteUsage(), B=ComponentMask(None=@n5) {r=true})
  )
  n7 = VectorParameter() {group=Boss_Derez, parameterName=PivotOffset, defaultValue=<rgba 0 0 0 1>}
  n8 = Divide(A=NamedRerouteDeclaration(None=Add(A=@n6, B=@n7.RGB)) {name=WorldPosition, nodeColor=<rgba 1 0.2812 0 1>}, B=ObjectLocalBounds().Max)
  NamedRerouteDeclaration(None=ChannelMaskParameter(None=@n8) {maskChannel=Blue, parameterName=FadeAxis, defaultValue=<rgba 0 0 1 0>}.R) {name=Transition_Gradient, nodeColor=<rgba 0 0.8359 1 1>}
  n9 = Subtract(A=NamedRerouteUsage(), B=ScalarParameter() {defaultValue=0.2, group=Boss_Derez, parameterName=DebugThreshold_Value})
  n10 = Constant() {r=1}
  n11 = NamedRerouteUsage()
  n12 = Multiply(
   A=MaterialFunctionCall(
    Transition Phase=MaterialFunctionCall!(ThresholdValue=@n9, ThresholdRange=@n10, Alpha=@n11, TwoSidedThreshold) {materialFunction=/Engine/Functions/Engine_MaterialFunctions02/Masking/ThresholdWithRange}.Result
    Height Texture=NamedRerouteUsage()
    Contrast=Constant() {r=1}
   ) {materialFunction=&HeightLerp}.Lerp Alpha No Contrast
   B=ScalarParameter() {defaultValue=2.5, group=Boss_Derez, parameterName=BColorEmissive_Strength}
  )
  NamedRerouteDeclaration(None=Saturate(None=@n12)) {name=BaseColorMAsk, nodeColor=<rgba 1 0.2109 0 1>}
  NamedRerouteDeclaration(None=@n3.Specular) {name=Input_Specular, nodeColor=<rgba 0.8516 0 1 1>}
  NamedRerouteDeclaration(None=@n3.Roughness) {name=Input_Roughness, nodeColor=<rgba 0 1 0.5313 1>}
  n13 = LocalPosition() {localOrigin=InstancePreSkinning}
  NamedRerouteDeclaration(None=@n13.XY) {name=PreSkin_XY, nodeColor=<rgba 0.0078 1 0 1>}
  n14 = Multiply(A=Time(), B=VectorParameter() {group=Boss_Derez, parameterName=BaseNoisePanSpeedAndDirection, defaultValue=<rgba 6 6 -6 1>}.RGB)
  n15 = TextureObject() {texture=/Game/Global/Textures/Default/Tileable_Patterns/T_Blowing_Smoke_Pattern2_Tile_DISP}
  n16 = ScalarParameter() {defaultValue=24, group=Boss_Derez, parameterName=Distortion_Size}
  n17 = Add(
   A=NamedRerouteUsage()
   B=Multiply(A=Time(), B=VectorParameter() {group=Boss_Derez, parameterName=DistortionPanSpeedAndDirection, defaultValue=<rgba 12 12 -12 1>}.RGB)
  )
  n18 = Add(
   A=NamedRerouteUsage()
   B=Multiply(
    A=MaterialFunctionCall(TextureObject=@n15, TextureSize=@n16, WorldPosition=@n17) {materialFunction=&WorldAlignedTexture}.XYZ Texture
    B=ScalarParameter() {defaultValue=0.25, group=Boss_Derez, parameterName=Distortion_Strength}
   )
  )
  NamedRerouteDeclaration(None=MaterialFunctionCall(
    TextureObject=TextureObject() {texture=&T_Blowing_Smoke_Pattern2_Tile_DISP}
    TextureSize=ScalarParameter() {defaultValue=24, group=Boss_Derez, parameterName=Noise_Size}
    WorldPosition=Add(A=@n18, B=@n14)
   ) {materialFunction=&WorldAlignedTexture}.XYZ Texture) {name=Noise, nodeColor=<rgba 0 0.5547 1 1>}
  n19 = ScalarParameter() {defaultValue=5, group=Boss_Derez, parameterName=VertExpand}
  n20 = LinearInterpolate(
   A=ScalarParameter() {defaultValue=1, group=Boss_Derez, parameterName=WPO_MinScale}
   B=ScalarParameter() {defaultValue=1, group=Boss_Derez, parameterName=WPO_MaxScale}
   Alpha=NamedRerouteUsage()
  ) {constA=0.1}
  n21 = Multiply(A=Multiply(A=VertexNormalWS(), B=Multiply(A=@n19)), B=@n20)
  n22 = NamedRerouteUsage()
  n23 = Multiply(A=Multiply(A=VectorParameter() {group=Boss_Derez, parameterName=ExplodeSmear, defaultValue=<rgba 0 0 10 1>}.RGB), B=@n20)
  n24 = NamedRerouteUsage()
  NamedRerouteDeclaration(None=Add(A=Multiply(A=@n21, B=@n22), B=Multiply(A=@n23, B=@n24))) {name=WPO, nodeColor=<rgba 0.6875 1 0 1>}
  NamedRerouteDeclaration(None=@n13.Z) {name=PreSkin_Z, nodeColor=<rgba 1 0 0.3281 1>}
  n25 = Subtract(A=Add(A=NamedRerouteUsage()), B=NamedRerouteUsage())
  n26 = Multiply(A=NamedRerouteUsage(), B=ScalarParameter() {defaultValue=1, group=Boss_Derez, parameterName=OpacityNoise_Str})
  NamedRerouteDeclaration(None=Saturate(None=Subtract(A=@n25, B=@n26))) {name=Opacity_Mask, nodeColor=<rgba 0 1 0.0391 1>}
  n27 = Subtract(A=Time(), B=ScalarParameter() {bUseCustomPrimitiveData=true, defaultValue=-999999986991104, parameterName=GameTimeFromBP})
  n28 = ScalarParameter() {defaultValue=1.8, parameterName=DerezDuration}
  n29 = OneMinus(None=Saturate(None=Divide(A=@n27, B=@n28)))
  n30 = ScalarParameter() {defaultValue=1, group=SpawnEffect, parameterName=DissolveGradientPercentage, sortPriority=31}
  NamedRerouteDeclaration(None=LinearInterpolate(Alpha=StaticSwitchParameter(True=@n29, False=@n30) {parameterName=UseDerezTimeStamp}) {constA=-1, constB=2.5}) {name=TransitionPhase, nodeColor=<rgba 0 1 0.4375 1>}
  NamedRerouteDeclaration(None=@n3.WorldPositionOffset) {name=Input_WPO, nodeColor=<rgba 1 0.1172 0 1>}
  n31 = Add(A=NamedRerouteUsage(), B=ScalarParameter() {defaultValue=0.15, group=Boss_Derez, parameterName=WPO_Threshold_Offset}) {constB=0.5}
  n32 = ScalarParameter() {defaultValue=1, group=Boss_Derez, parameterName=WPO_ThresholdRange}
  n33 = NamedRerouteUsage()
  NamedRerouteDeclaration(None=MaterialFunctionCall(
    Transition Phase=MaterialFunctionCall(ThresholdValue=@n31, ThresholdRange=@n32, Alpha=@n33) {materialFunction=&ThresholdWithRange}.Result
    Height Texture=NamedRerouteUsage()
    Contrast=ScalarParameter() {defaultValue=1, group=Boss_Derez, parameterName=WPO_Contrast}
   ) {materialFunction=&HeightLerp}.Lerp Alpha No Contrast) {name=WPO_Gradient, nodeColor=<rgba 0 0.2031 1 1>}
  n34 = VectorParameter() {group=Boss_Derez, parameterName=EmissiveColor2, defaultValue=<rgba 0.6724 0.0672 0.0797 1>}
  n35 = DotProduct(A=PixelNormalWS(), B=CameraVectorWS())
  n36 = Multiply(
   A=NamedRerouteUsage()
   B=LinearInterpolate(
    A=VectorParameter() {group=Boss_Derez, parameterName=EmissiveColor1, defaultValue=<rgba 1 0.2445 0.2355 1>}.RGB
    B=@n34.RGB
    Alpha=Saturate(None=@n35)
   )
  )
  n37 = ScalarParameter() {defaultValue=2.5, group=Boss_Derez, parameterName=Emissive_Strength}
  NamedRerouteDeclaration(None=Power(Base=Multiply(A=@n36, B=@n37))) {name=EmissiveColor, nodeColor=<rgba 0.875 0 1 1>}
  n38 = NamedRerouteUsage()
  n39 = ScalarParameter() {defaultValue=1, group=Boss_Derez, parameterName=ThresholdRange}
  n40 = NamedRerouteUsage()
  NamedRerouteDeclaration(None=MaterialFunctionCall(
    Transition Phase=MaterialFunctionCall(ThresholdValue=@n38, ThresholdRange=@n39, Alpha=@n40) {materialFunction=&ThresholdWithRange}.Result
    Height Texture=NamedRerouteUsage()
   ) {materialFunction=&HeightLerp}.Lerp Alpha No Contrast) {name=Emissive_Mask, nodeColor=<rgba 0.5703 1 0 1>}
  NamedRerouteDeclaration(None=@n3.BaseColor) {name=Input_BaseColor, nodeColor=<rgba 1 0 0.8906 1>}
  NamedRerouteUsage()
  OneMinus(None=@n8)
  n41 = Time()
  OneMinus(None=Frac(None=Multiply(A=@n41) {constB=0.2}))
  Power(Base=@n12)
  Power(Base=Multiply(A=@n35) {constB=2})
  Reroute(None=@n34.RGB)
  n42 = Add(A=NamedRerouteUsage())
  n43 = Add(A=NamedRerouteUsage(), B=ScalarParameter() {defaultValue=0.25, group=Boss_Derez, parameterName=WPO_GradientOffset})
  n44 = NamedRerouteUsage()
  n45 = ScalarParameter() {defaultValue=1, group=Boss_Derez, parameterName=WPONoise_Str}
  Saturate(None=Subtract(A=Subtract(A=@n42, B=@n43), B=Multiply(A=@n44, B=@n45)))
  ScalarParameter() {bUseCustomPrimitiveData=true, parameterName=TransitionPhase}
 ]}
