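# Master character material for the Wandering Boss, adding two glowing, hit-reactive weak points to the shared CharacterParent functions and a demon boss derez effect.
# Sphere masks at parameterized locations define weak points A and B, blending outer and inner colors.
# Custom primitive data timestamps and flags drive time-based show, hide and hit animations, with sine-pulsed hit intensity.
# The glow tints base color and adds AO-attenuated emissive through material attributes.
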Material /FirePetal_WanderingBoss/Assets/Materials/M_WanderingBoss_Base
graph {
 properties {
  bUseMaterialAttributes=true
  bUsedWithClothing=true
  bUsedWithMeshDeformer=true
  bUsedWithMorphTargets=true
  bUsedWithSkeletalMesh=true
  pixelDepthOffsetMode=PDOM_Legacy
  shadingModel=MSM_FromMaterialExpression
  usedShadingModels=Default Lit | Subsurface | Clear Coat | Subsurface Profile | Cloth
 }
 nodes [
  Constant() {r=1}
  n1 = MaterialFunctionCall() {materialFunction=&MF_CharacterParent_2023_PreFX}
  MaterialFunctionCall!(Material=MaterialFunctionCall(
    Material=NamedRerouteUsage()
    M=NamedRerouteDeclaration(None=@n1.M) {name=Mask, nodeColor=<rgba 0.0078 1 0 1>}
    Pre Skinned Local Position=NamedRerouteDeclaration(None=@n1.Pre skinned Local Position) {name=PreSkinnedLocalPos, nodeColor=<rgba 0.6406 0 1 1>}
    Distance Blend=@n1.Distance Blend
    Fresnel=NamedRerouteDeclaration(None=@n1.Fresnel) {name=Fresnel, nodeColor=<rgba 0 1 0.3203 1>}
   ) {materialFunction=&MF_CharacterParent_2023_PostFX}.Result) {materialFunction=/FirePetalBosses/Assets/Materials/MF_DemonBoss_Derez}
  NamedRerouteDeclaration(None=@n1.Result) {name=MA_In, nodeColor=<rgba 0 0.7187 1 1>}
  n2 = GetMaterialAttributes(None=NamedRerouteUsage()) {attributeGetTypes [#1, #2]}
  n3 = ScalarParameter() {defaultValue=10, group=WeakPoints, parameterName=HitEffectIntensity}
  n4 = NamedRerouteUsage()
  n5 = ConstantBiasScale(None=Sine(None=Add(A=@n4) {constB=0.75}))
  n6 = Add(A=Multiply(A=NamedRerouteUsage(), B=Multiply(A=@n3, B=@n5) {constA=10}))
  n7 = SphereMask(
   A=NamedRerouteUsage()
   B=VectorParameter() {group=WeakPoints, parameterName=WeakPointA_Location, defaultValue=<rgba 70 9 85 1>}.RGB
   Radius=ScalarParameter() {defaultValue=55, group=WeakPoints, parameterName=WeakPointA_Radius}
   Hardness=ScalarParameter() {group=WeakPoints, parameterName=WeakPointA_Hardness}
  ) {hardnessPercent=0}
  n8 = NamedRerouteUsage()
  n9 = Subtract(
   A=Subtract(
    A=MaterialFunctionCall!(In=@n7) {materialFunction=/Game/Effects/Fort_Effects/Materials/_Functions/MF_Smoothstep_01}.Result
    B=OneMinus(None=@n8)
   )
   B=NamedRerouteUsage()
  )
  n10 = LinearInterpolate(
   A=VectorParameter() {parameterName=WeakPoint_OuterColor, defaultValue=<rgba 1 0 0 1>}.RGB
   B=VectorParameter() {parameterName=WeakPoint_InnerColor, defaultValue=<rgba 1 0.1498 0.01 1>}.RGB
   Alpha=NamedRerouteDeclaration(None=Saturate(None=@n9)) {name=WeakpointAMask, nodeColor=<rgba 0 1 0.1797 1>}
  )
  n11 = SphereMask(
   A=NamedRerouteUsage()
   B=VectorParameter() {group=WeakPoints, parameterName=WeakPointB_Location, defaultValue=<rgba -70 9 85 1>}.RGB
   Radius=ScalarParameter() {defaultValue=55, group=WeakPoints, parameterName=WeakPointB_Radius}
   Hardness=ScalarParameter() {group=WeakPoints, parameterName=WeakPointB_Hardness}
  ) {hardnessPercent=0}
  n12 = NamedRerouteUsage()
  n13 = Subtract(
   A=Subtract(
    A=MaterialFunctionCall(In=@n11) {materialFunction=&MF_Smoothstep_01}.Result
    B=OneMinus(None=@n12)
   )
   B=NamedRerouteUsage()
  )
  n14 = LinearInterpolate(
   A=VectorParameter() {parameterName=WeakPoint_OuterColor, defaultValue=<rgba 1 0 0 1>}.RGB
   B=VectorParameter() {parameterName=WeakPoint_InnerColor, defaultValue=<rgba 1 0.1498 0.01 1>}.RGB
   Alpha=NamedRerouteDeclaration(None=Saturate(None=@n13)) {name=WeakpointBMask, nodeColor=<rgba 0.5 0 1 1>}
  )
  n15 = ScalarParameter() {defaultValue=10, group=WeakPoints, parameterName=HitEffectIntensity}
  n16 = NamedRerouteUsage()
  n17 = ConstantBiasScale(None=Sine(None=Add(A=@n16) {constB=0.75}))
  n18 = Add(A=Multiply(A=Multiply(A=@n15, B=@n17) {constA=10}, B=NamedRerouteUsage()))
  n19 = Multiply(A=Max(A=Multiply(A=@n6, B=@n10), B=Multiply(A=@n14, B=@n18)), B=Max(A=NamedRerouteUsage(), B=NamedRerouteUsage()))
  n20 = Add(A=@n2.Base Color, B=Multiply(A=@n2.Base Color, B=@n19))
  n21 = NamedRerouteUsage()
  n22 = Saturate(None=Multiply(A=ComponentMask(None=@n21) {r=true}, B=ScalarParameter() {defaultValue=0.5, group=WeakPoints, parameterName=AOContribution}))
  n23 = Multiply(A=@n19, B=ScalarParameter() {defaultValue=1, group=WeakPoints, parameterName=WeakPoint_Intensity})
  n24 = Add(A=@n2.Emissive Color, B=Multiply(A=OneMinus(None=@n22), B=Multiply(A=@n2.Base Color, B=@n23)))
  NamedRerouteDeclaration(None=BlendMaterialAttributes(
    A=NamedRerouteUsage()
    B=SetMaterialAttributes!(MaterialAttributes=@n2.MaterialAttributes, Base Color=@n20, Emissive Color=@n24) {attributeSetTypes [#1, #2]}
    Alpha=Constant() {r=1}
   )) {name=MA_Out, nodeColor=<rgba 1 0.9609 0 1>}
  n25 = ScalarParameter() {bUseCustomPrimitiveData=true, parameterName=HitEffectA_Timestamp_CPD, primitiveDataIndex=7}
  n26 = Subtract(A=Time(), B=@n25)
  n27 = Saturate(None=Divide(A=Abs(None=@n26), B=ScalarParameter() {defaultValue=1, group=WeakPoints, parameterName=HitEffectSpeed}))
  NamedRerouteDeclaration(None=Multiply(A=OneMinus(None=@n27), B=Saturate(None=@n25))) {name=HitEffectATime, nodeColor=<rgba 1 0.2109 0 1>}
  n28 = ScalarParameter() {bUseCustomPrimitiveData=true, parameterName=HitEffectB_Timestamp_CPD, primitiveDataIndex=8}
  n29 = Subtract(A=Time(), B=@n28)
  n30 = Saturate(None=Divide(A=Abs(None=@n29), B=ScalarParameter() {defaultValue=1, group=WeakPoints, parameterName=HitEffectSpeed}))
  NamedRerouteDeclaration(None=Multiply(A=OneMinus(None=@n30), B=Saturate(None=@n28))) {name=HitEffectBTime, nodeColor=<rgba 1 0.2109 0 1>}
  n31 = ScalarParameter() {bUseCustomPrimitiveData=true, parameterName=WeakpointB_Timestamp_CPD, primitiveDataIndex=6}
  n32 = Subtract(A=Time(), B=@n31)
  n33 = ScalarParameter() {defaultValue=1, group=WeakPoints, parameterName=HideAnimSpeed}
  n34 = ScalarParameter() {defaultValue=1, group=WeakPoints, parameterName=ShowAnimSpeed}
  n35 = ScalarParameter() {bUseCustomPrimitiveData=true, parameterName=WeakpointB_HideFlag_CPD, primitiveDataIndex=5}
  n36 = Saturate(None=Divide(A=Abs(None=@n32), B=LinearInterpolate(A=@n33, B=@n34, Alpha=@n35)))
  n37 = OneMinus(None=@n36)
  n38 = ScalarParameter() {bUseCustomPrimitiveData=true, parameterName=WeakpointB_HideFlag_CPD, primitiveDataIndex=5}
  NamedRerouteDeclaration(None=Multiply(A=LinearInterpolate(A=@n37, B=@n36, Alpha=@n38), B=Saturate(None=@n31))) {name=WeakpointBTime, nodeColor=<rgba 1 0.2109 0 1>}
  n39 = ScalarParameter() {bUseCustomPrimitiveData=true, parameterName=WeakpointA_Timestamp_CPD, primitiveDataIndex=4}
  n40 = Subtract(A=Time(), B=@n39)
  n41 = ScalarParameter() {defaultValue=1, group=WeakPoints, parameterName=HideAnimSpeed}
  n42 = ScalarParameter() {defaultValue=1, group=WeakPoints, parameterName=ShowAnimSpeed}
  n43 = ScalarParameter() {bUseCustomPrimitiveData=true, parameterName=WeakpointA_HideFlag_CPD, primitiveDataIndex=3}
  n44 = Saturate(None=Divide(A=Abs(None=@n40), B=LinearInterpolate(A=@n41, B=@n42, Alpha=@n43)))
  n45 = OneMinus(None=@n44)
  n46 = ScalarParameter() {bUseCustomPrimitiveData=true, parameterName=WeakpointA_HideFlag_CPD, primitiveDataIndex=3}
  NamedRerouteDeclaration(None=Multiply(A=LinearInterpolate(A=@n45, B=@n44, Alpha=@n46), B=Saturate(None=@n39))) {name=WeakpointATime, nodeColor=<rgba 1 0.2109 0 1>}
  n47 = LocalPosition() {localOrigin=InstancePreSkinning}
  NamedRerouteDeclaration(None=SphereMask(A=VertexInterpolator(VS=@n47.XYZ).PS, B=Constant3Vector() {constant=<rgba 0 0 80 0>}) {attenuationRadius=50, hardnessPercent=10}) {name=Masked, nodeColor=<rgba 0 1 0.5078 1>}
 ]
}
